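# Material function that projects a panning normal map biplanarly, stored with the Pickaxe_PrismParticle cosmetic.
# Position divided by TextureSize, offset by Pan, samples one texture on YZ and XZ planes using Masks samplers.
# Each sample is swizzled and combined with the input Normal, then blended by contrast-adjusted absolute normal Y.
# The normalized result outputs as local space and, via a local-to-tangent transform, tangent space.
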MaterialFunction /BRCosmetics/Weapons/FORT_Melee/Pickaxe_PrismParticle/Materials/MF_BiplanarNormal_Pan
graph {
 properties {description=Biplanar projection for Normal map}
 nodes [
  n1 = Divide(
   A=FunctionInput(Preview=WorldPosition().XYZ) {inputName=Position, sortPriority=1}
   B=FunctionInput(Preview=Constant3Vector() {constant=<rgba 128 128 128 0>}) {inputName=TextureSize, sortPriority=3}
  )
  n2 = FunctionInput() {bUsePreviewValueAsDefault=true, inputName=Pan, inputType=FunctionInput_Vector2, sortPriority=4}
  n3 = FunctionInput(Preview=TextureObject() {samplerType=SAMPLERTYPE_Normal, texture=/Game/Characters/Player/Male/Large/Bodies/M_LRG_Slurp_Monster/Textures/T_SlurpBubble_N}) {inputName=Texture, inputType=FunctionInput_Texture2D}
  n4 = TextureSample(UVs=Add(A=ComponentMask(None=@n1) {b=true, g=true}, B=@n2), Tex=@n3) {
   desc=X axis
   samplerSource=SSM_Wrap_WorldGroupSettings
   samplerType=SAMPLERTYPE_Masks
   texture=/Game/Weapons/WeaponSkins/Textures/T_Swirly_Noise_highContrast
  }
  n5 = FunctionInput(Preview=VertexNormalWS()) {bUsePreviewValueAsDefault=true, inputName=Normal, sortPriority=2, previewValue=<0 0 1 1>}
  n6 = ComponentMask(None=@n5) {r=true}
  n7 = ComponentMask(None=@n5) {g=true}
  n8 = ComponentMask(None=@n5) {b=true}
  MaterialExpressionAppend3Vector_0 = MaterialXAppend3Vector!(A=Multiply(A=@n4.B, B=@n6) {constB=-1}, B=Add(A=@n4.R, B=@n7), C=Add(A=@n4.G, B=@n8))
  n9 = TextureSample(UVs=Add(A=ComponentMask(None=@n1) {b=true, r=true}, B=@n2), Tex=@n3) {
   samplerSource=SSM_Wrap_WorldGroupSettings
   samplerType=SAMPLERTYPE_Masks
   texture=&T_Swirly_Noise_highContrast
  }
  MaterialExpressionAppend3Vector_1 = MaterialXAppend3Vector(A=Add(A=@n9.R, B=@n6), B=Multiply(A=@n9.B, B=@n7), C=Add(A=@n9.G, B=@n8))
  n10 = ComponentMask(None=@n5) {g=true}
  n11 = Constant()
  n12 = LinearInterpolate(
   A=@MaterialExpressionAppend3Vector_0
   B=@MaterialExpressionAppend3Vector_1
   Alpha=MaterialFunctionCall(In=Abs(None=@n10), Contrast=FunctionInput(Preview=@n11) {inputName=BlendContrast, inputType=FunctionInput_Scalar, sortPriority=5}) {materialFunction=&CheapContrast}.Result
  )
  FunctionOutput(None=Transform(None=Normalize(VectorInput=@n12)) {transformSourceType=TRANSFORMSOURCE_Local, transformType=TRANSFORM_Tangent}) {outputName=TangentSpaceNormal}
  FunctionOutput(None=Normalize(VectorInput=@n12)) {outputName=LocalSpaceNormal}
 ]
}
