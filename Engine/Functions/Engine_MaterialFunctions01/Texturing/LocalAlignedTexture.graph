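# LocalAlignedTexture is a library material function that projects one input texture by blending three samples by surface direction, outputting XYZ, XY and Z results.
# It samples UV channels 0, 1 and 2, using contrast-boosted absolute X and Z of the world vertex normal as lerp weights.
# Despite its worldspace-tiling description, it computes no position-based UVs, and its Normal input and world-to-local Transform feed nothing.
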MaterialFunction /Engine/Functions/Engine_MaterialFunctions01/Texturing/LocalAlignedTexture
graph {
 properties {bExposeToLibrary=true, description=Tiles a texture in worldspace, libraryCategoriesText [Misc,Texturing]}
 nodes [
  FunctionInput() {bUsePreviewValueAsDefault=true, inputName=Normal, previewValue=<0 0 1 1>}
  n1 = FunctionInput(Preview=TextureObject() {texture=/Engine/EngineMaterials/DefaultDiffuse}) {inputName=TextureObject, inputType=FunctionInput_Texture2D}
  n2 = VertexNormalWS()
  n3 = ComponentMask(None=@n2) {r=true}
  n4 = Constant() {r=1}
  n5 = LinearInterpolate(
   A=TextureSample(UVs=TextureCoordinate() {coordinateIndex=1}, Tex=@n1).RGB
   B=TextureSample(UVs=TextureCoordinate(), Tex=@n1).RGB
   Alpha=MaterialFunctionCall(In=Abs(None=@n3), Contrast=@n4) {materialFunction=&CheapContrast}.Result
  )
  n6 = TextureSample(UVs=TextureCoordinate() {coordinateIndex=2}, Tex=@n1)
  n7 = Abs(None=ComponentMask(None=@n2) {b=true})
  FunctionOutput(None=LinearInterpolate(
    A=@n5
    B=@n6.RGB
    Alpha=MaterialFunctionCall(In=@n7, Contrast=@n4) {materialFunction=&CheapContrast}.Result
   )) {outputName=XYZ Texture, sortPriority=1}
  FunctionOutput(None=@n5) {outputName=XY Texture, sortPriority=1}
  FunctionOutput(None=@n6.RGB) {outputName=Z Texture, sortPriority=1}
  Transform(None=@n2) {transformSourceType=TRANSFORMSOURCE_World, transformType=TRANSFORM_Local}
 ]
}
