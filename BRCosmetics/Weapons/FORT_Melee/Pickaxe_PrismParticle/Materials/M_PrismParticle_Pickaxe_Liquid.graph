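# Liquid material for the PrismParticle pickaxe, setting Subsurface shading between the shared character PreFX and PostFX functions.
# Three panning biplanar layers of Slurp bubble textures build masks and normals, faded by vertical normal.
# Emissive mixes view-aligned fake reflection, inverted fresnel, a sphere-mask inner glow and bubble color.
# Fresnel drives base color and subsurface opacity; several branches, including older world-aligned bubbles and sparkle brightness, are left unconnected.
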Material /BRCosmetics/Weapons/FORT_Melee/Pickaxe_PrismParticle/Materials/M_PrismParticle_Pickaxe_Liquid
graph {
 properties {
  bScreenSpaceReflections=true
  bTangentSpaceNormal=false
  bUseMaterialAttributes=true
  bUsedWithClothing=true
  bUsedWithMeshDeformer=true
  bUsedWithNanite=true
  bUsedWithSkeletalMesh=true
  pixelDepthOffsetMode=PDOM_Legacy
  refractionMethod=RM_IndexOfRefraction
  shadingModel=MSM_FromMaterialExpression
  translucencyDirectionalLightingIntensity=5
  translucencyLightingMode=TLM_VolumetricPerVertexDirectional
  usedShadingModels=Default Lit | Subsurface | Clear Coat | Subsurface Profile | Cloth
 }
 nodes [
  n1 = MaterialFunctionCall() {materialFunction=&MF_CharacterParent_2023_PreFX}
  ComponentMask(None=@n1.M) {r=true}
  n2 = MaterialFunctionCall!(ReflectionVector, ReflectionTexture) {materialFunction=/Engine/Functions/Engine_MaterialFunctions01/Reflections/ViewAlignedReflection}
  n3 = ScalarParameter() {group=PrismParticle, parameterName=PP_ReflectionMipBias}
  n4 = PixelNormalWS()
  n5 = StaticBool() {value=true}
  n6 = ScalarParameter() {defaultValue=2, group=PrismParticle, parameterName=PP_InvertedFresnelPower}
  n7 = Multiply(
   A=LinearInterpolate(
    B=TextureSample!(UVs=@n2.UVs, Tex, Bias=@n3, Apply View MipBias) {mipValueMode=TMVM_MipBias, texture=/Game/Packages/Fortress_SharedMaterials/Textures/T_Ravens_SpecularReflection}.RGB
    Alpha=ScalarParameter() {defaultValue=0.2, parameterName=PP_ReflectionIntensity}
   ) {constB=0}
   B=Multiply(
    A=ScalarParameter() {defaultValue=4, group=PrismParticle, parameterName=PP_InvertedFresnelBrightness}
    B=MaterialFunctionCall(Normal Vector=@n4, Invert Fresnel=@n5, Power=@n6) {materialFunction=&Fresnel_Function}.Result
   )
  )
  n8 = Add() {constA=0.5, constB=0}
  n9 = ScalarParameter() {defaultValue=2, group=PrismParticle, parameterName=PP_GlowAmount}
  n10 = NamedRerouteUsage()
  n11 = Add(B=Multiply(A=OneMinus(None=@n10), B=Constant() {r=0.5})) {constA=1}
  n12 = VectorParameter() {group=PrismParticle, parameterName=PP_2DGlowColor, defaultValue=<rgba 0.0509 0.2229 0.6979 2>}
  n13 = Multiply(A=Multiply(A=Add(A=@n7, B=@n8), B=Multiply(A=@n9, B=@n11)), B=@n12.RGB)
  n14 = LinearInterpolate(
   A=Add(
    A=Power(Base=@n13, Exp=@n12.A)
    B=VectorParameter() {group=PrismParticle, parameterName=PP_GlowPass, defaultValue=<rgba 0 0.0314 0.27 0.1>}.RGB
   )
   B=VectorParameter() {group=PrismParticle, parameterName=PP_BubbleColor, defaultValue=<rgba 0.0436 0.0725 0.3438 1>}.RGB
   Alpha=NamedRerouteUsage()
  )
  n15 = NamedRerouteUsage()
  LinearInterpolate(A=@n14, Alpha=@n15)
  n16 = SetMaterialAttributes!(
   MaterialAttributes=@n1.Result
   Base Color=NamedRerouteUsage()
   Metallic=ScalarParameter() {group=PrismParticle, parameterName=PP_Metallic}
   Specular=ScalarParameter() {defaultValue=0.1514, group=PrismParticle, parameterName=PP_Specular}
   Roughness=ScalarParameter() {defaultValue=0.1101, group=PrismParticle, parameterName=PP_Roughness}
   Emissive Color=NamedRerouteUsage()
   Normal=NamedRerouteUsage()
   World Position Offset
   Shading Model=ShadingModel() {shadingModel=MSM_Subsurface}
   Opacity=NamedRerouteUsage()
   Fuzz Color or Subsurface Color=VectorParameter() {group=PrismParticle, parameterName=PP_SubSurfaceColor, defaultValue=<rgba 0 0.0688 0.1719 1>}.RGB
   Refraction=LinearInterpolate(Alpha=NamedRerouteUsage()) {constA=1, constB=5}
  ) {attributeSetTypes [
    #1
    #2
    #3
    #4
    #5
    #6
    #7
    #8
    #9
    #10
    #11
   ]}
  MaterialFunctionCall(
   Material=@n16
   M=@n1.M
   Pre Skinned Local Position=NamedRerouteDeclaration(None=@n1.Pre skinned Local Position) {name=PreskinLocalPos, nodeColor=<rgba 0.0781 1 0 1>}
   Distance Blend=@n1.M
   Fresnel=@n1.M
  ) {materialFunction=&MF_CharacterParent_2023_PostFX}
  MaterialFunctionCall!(Normal, TextureObject) {materialFunction=/Engine/Functions/Engine_MaterialFunctions01/Texturing/LocalAlignedTexture}
  n17 = MaterialFunctionCall() {materialFunction=&ViewAlignedReflection}
  n18 = ScalarParameter() {defaultValue=0.5, parameterName=ViewDependantDistortion}
  n19 = ScalarParameter() {defaultValue=0.05, parameterName=water_SpeedX}
  n20 = ScalarParameter() {defaultValue=0.1, parameterName=water_SpeedY}
  n21 = TextureSample!(UVs=Panner(Coordinate=Multiply(A=@n17.UVs, B=@n18), Speed=AppendVector(A=@n19, B=@n20)), Tex, Apply View MipBias) {samplerType=SAMPLERTYPE_LinearColor, texture=/Game/Characters/Player/Male/Medium/Heads/M_MED_AUX_Alien_Head_01/Textures/T_DistortionMask}
  n22 = TextureObject() {samplerType=SAMPLERTYPE_Masks, texture=/Game/Characters/Player/Male/Medium/Bodies/M_Med_Commando_AlienFishHead/Textures/T_SimpleBubble}
  n23 = ObjectPositionWS()
  n24 = Constant3Vector() {constant=<rgba -15 0 19 0>}
  n25 = Subtract(
   A=WorldPosition().XYZ
   B=LinearInterpolate(
    A=VectorParameter() {parameterName=PP_PivotLocation, defaultValue=<rgba 0 0 0 1>}.RGB
    B=Add(A=@n23, B=@n24)
    Alpha=ScalarParameter() {defaultValue=1, parameterName=PP_UseObjectPosition}
   )
  )
  n26 = ScalarParameter() {parameterName=Bubble Speed X}
  n27 = ScalarParameter() {defaultValue=0.15, parameterName=Bubble Speed Y_1}
  n28 = MaterialFunctionCall(Float3=MaterialFunctionCall!(
    TextureObject=@n22
    TextureSize=ScalarParameter() {defaultValue=10, parameterName=BubbleSize_A}
    WorldPosition=@n25
    Export Float 4
    In=AppendVector(A=@n26, B=@n27)
    World Space Normal
    ProjectionTransitionContrast
   ) {materialFunction=/Game/Athena/SafeZone/WorldAlignedTexture_ScalePan}.XYZ Texture) {materialFunction=&BreakOutFloat3Components}
  n29 = TextureObject() {samplerType=SAMPLERTYPE_Masks, texture=&T_SimpleBubble}
  n30 = ScalarParameter() {defaultValue=0.3, parameterName=Bubble Speed Y_2}
  n31 = MaterialFunctionCall(Float3=MaterialFunctionCall(
    TextureObject=@n29
    TextureSize=ScalarParameter() {defaultValue=20, parameterName=BubbleSize_A_1}
    WorldPosition=@n25
    In=AppendVector(A=@n26, B=@n30)
   ) {materialFunction=&WorldAlignedTexture_ScalePan}.XYZ Texture) {materialFunction=&BreakOutFloat3Components}
  n32 = TextureObject() {samplerType=SAMPLERTYPE_Masks, texture=&T_SimpleBubble}
  n33 = MaterialFunctionCall(
   TextureObject=@n32
   TextureSize=ScalarParameter() {defaultValue=10, parameterName=BubbleSize_A_2}
   WorldPosition=@n25
   In=AppendVector(A=@n26, B=ScalarParameter() {defaultValue=0.421, parameterName=Bubble Speed Y_3})
  ) {materialFunction=&WorldAlignedTexture_ScalePan}
  Multiply(
   A=LinearInterpolate(B=@n21.G) {constA=1}
   B=Add(
    A=Add(A=@n28.R, B=@n31.G)
    B=MaterialFunctionCall(Float3=@n33.XYZ Texture) {materialFunction=&BreakOutFloat3Components}.B
   )
  )
  n34 = MaterialFunctionCall!(
   Scale=ScalarParameter() {defaultValue=-0.1, group=PrismParticle, parameterName=PP_SparkleScale}
   Sparkle World center point=ObjectPositionWS()
  ) {materialFunction=/Game/Packages/Fortnite_UniqueMaterials_1/Effects/FortniteSparkle}
  Multiply(A=@n34.Result, B=ScalarParameter() {group=PrismParticle, parameterName=PP_SparkleBrightness})
  n35 = VertexNormalWS()
  n36 = NamedRerouteUsage()
  n37 = NamedRerouteUsage()
  n38 = NamedRerouteUsage()
  n39 = NamedRerouteUsage()
  n40 = Add(
   A=Add(A=LinearInterpolate(A=@n35, B=@n36, Alpha=@n37), B=LinearInterpolate(A=@n35, B=@n38, Alpha=@n39))
   B=LinearInterpolate(A=@n35, B=NamedRerouteUsage(), Alpha=NamedRerouteUsage())
  )
  Multiply(A=@n40)
  NamedRerouteDeclaration(None=Add(A=@n14, B=@n15)) {name=Emissive_OUT, nodeColor=<rgba 1 0.5156 0 1>}
  n41 = Fresnel() {baseReflectFraction=0, exponent=0.5}
  n42 = Multiply(A=OneMinus(None=Clamp(None=@n41)), B=NamedRerouteUsage())
  NamedRerouteDeclaration(None=SphereMask(
    A=Saturate(None=@n42)
    B=ScalarParameter() {defaultValue=0.2, parameterName="Bubble Highlight "}
    Radius=ScalarParameter() {defaultValue=0.5, parameterName=Bubble Highlight Thickness}
    Hardness=Constant() {r=0.5}
   )) {name=Bubble, nodeColor=<rgba 0 1 0.625 1>}
  n43 = Multiply(A=Time(), B=ScalarParameter() {parameterName=Bubble_ViewDistortion})
  n44 = Panner(
   Coordinate=Multiply(
    A=MaterialFunctionCall() {materialFunction=&ViewAlignedReflection}.UVs
    B=ScalarParameter() {defaultValue=1, group=Slurp, parameterName=Bubble_ViewDependantDistortion}
   )
   Time=@n43
   Speed=AppendVector(
    A=ScalarParameter() {defaultValue=0.1, group=Slurp, parameterName=Bubble_ViewDistortion_SpeedX}
    B=ScalarParameter() {defaultValue=1, group=Slurp, parameterName=Bubble_ViewDistortion_SpeedY}
   )
  )
  NamedRerouteDeclaration(None=TextureSample(UVs=Rotator(Coordinate=@n44, Time=@n43) {speed=0.0001}) {constMipValue=0, samplerType=SAMPLERTYPE_Grayscale, texture=&LowResBlurredNoise}.G) {name=Bubble_Distort, nodeColor=<rgba 0 1 0.8828 1>}
  n45 = PreSkinnedPosition()
  NamedRerouteDeclaration(None=ShaderStageSwitch(PixelShader=VertexInterpolator(VS=@n45).PS, VertexShader=@n45)) {name=PreSkinnedPos_VS-PS, nodeColor=<rgba 0 0.4844 1 1>}
  n46 = WorldPosition()
  n47 = ObjectPositionWS()
  NamedRerouteDeclaration(None=Add(A=Subtract(A=@n46.XYZ, B=@n47), B=NamedRerouteUsage())) {name=Bubble_PositionSample, nodeColor=<rgba 0.3359 1 0 1>}
  NamedRerouteDeclaration(None=VertexNormalWS()) {name=Bubble_NormalSample, nodeColor=<rgba 0.0078 0 1 1>}
  n48 = TextureObjectParameter() {
   group=Slurp
   parameterName=BubbleMask
   samplerType=SAMPLERTYPE_Masks
   texture=/Game/Characters/Player/Male/Large/Bodies/M_LRG_Slurp_Monster/Textures/T_SlurpBubble
  }
  NamedRerouteDeclaration(None=@n48) {name=Bubble_MaskTex, nodeColor=<rgba 0 0.5547 1 1>}
  n49 = TextureObjectParameter() {
   group=Slurp
   parameterName=BubbleNormal
   samplerType=SAMPLERTYPE_Normal
   texture=/Game/Characters/Player/Male/Large/Bodies/M_LRG_Slurp_Monster/Textures/T_SlurpBubble_N
  }
  NamedRerouteDeclaration(None=@n49) {name=Bubble_NormalTex, nodeColor=<rgba 0 0.2734 1 1>}
  n50 = Constant() {r=-2}
  n51 = VertexNormalWS()
  n52 = ComponentMask(None=@n51) {b=true}
  n53 = Saturate(None=MaterialFunctionCall(In=Abs(None=@n52), Contrast=ScalarParameter() {defaultValue=1, parameterName=TopBottomMask_Contrast}) {materialFunction=&CheapContrast}.Result)
  NamedRerouteDeclaration(None=LinearInterpolate(
    A=MaterialFunctionCall(Normal=@n40, Flatness=@n50) {materialFunction=&FlattenNormal}.Result
    B=@n51
    Alpha=@n53
   ) {constB=0}) {name=Normal_OUT, nodeColor=<rgba 1 0.5859 0 1>}
  NamedRerouteDeclaration(None=ScalarParameter() {defaultValue=3, parameterName=Contrast}) {name=BiplanarBlendContrast, nodeColor=<rgba 0.875 0 1 1>}
  n54 = NamedRerouteUsage()
  n55 = NamedRerouteUsage()
  n56 = NamedRerouteUsage()
  n57 = ScalarParameter() {defaultValue=25, parameterName=Bubble_Size_3}
  n58 = Multiply(A=VectorParameter() {parameterName=Bubble_Speed_3, defaultValue=<rgba -0.02 -0.75 0 1>}.RGB, B=Time())
  n59 = NamedRerouteUsage()
  NamedRerouteDeclaration(None=ComponentMask(None=MaterialFunctionCall!(Texture=@n54, Position=@n55, Normal=@n56, TextureSize=@n57, Pan=@n58, BlendContrast=@n59) {materialFunction=&/MF_Biplanar_Pan}.Result) {b=true}) {name=BubbleMask_3, nodeColor=<rgba 0.8516 1 0 1>}
  n60 = NamedRerouteUsage()
  n61 = NamedRerouteUsage()
  n62 = NamedRerouteUsage()
  n63 = ScalarParameter() {defaultValue=25, parameterName=Bubble_Size_2}
  n64 = Multiply(A=VectorParameter() {parameterName=Bubble_Speed_2, defaultValue=<rgba -0.094 -0.5 0 1>}.RGB, B=Time())
  n65 = NamedRerouteUsage()
  NamedRerouteDeclaration(None=ComponentMask(None=MaterialFunctionCall(Texture=@n60, Position=@n61, Normal=@n62, TextureSize=@n63, Pan=@n64, BlendContrast=@n65) {materialFunction=&MF_Biplanar_Pan}.Result) {g=true}) {name=BubbleMask_2, nodeColor=<rgba 0.2188 0 1 1>}
  NamedRerouteDeclaration(None=MaterialFunctionCall!(
    Texture=NamedRerouteUsage()
    Position=NamedRerouteUsage()
    Normal=NamedRerouteUsage()
    TextureSize=@n63
    Pan=@n64
    BlendContrast=NamedRerouteUsage()
   ) {materialFunction=&/MF_BiplanarNormal_Pan}.LocalSpaceNormal) {name=BubbleNormal_2, nodeColor=<rgba 0 1 0.2266 1>}
  n66 = NamedRerouteUsage()
  n67 = NamedRerouteUsage()
  n68 = NamedRerouteUsage()
  n69 = ScalarParameter() {defaultValue=15, parameterName=Bubble_Size_1}
  n70 = Multiply(A=VectorParameter() {parameterName=Bubble_Speed_1, defaultValue=<rgba 0.1 -0.3 0 1>}.RGB, B=Time())
  n71 = NamedRerouteUsage()
  NamedRerouteDeclaration(None=ComponentMask(None=MaterialFunctionCall(Texture=@n66, Position=@n67, Normal=@n68, TextureSize=@n69, Pan=@n70, BlendContrast=@n71) {materialFunction=&MF_Biplanar_Pan}.Result) {r=true}) {name=BubbleMask_1, nodeColor=<rgba 0.2188 0 1 1>}
  NamedRerouteDeclaration(None=MaterialFunctionCall(
    Texture=NamedRerouteUsage()
    Position=NamedRerouteUsage()
    Normal=NamedRerouteUsage()
    TextureSize=@n69
    Pan=@n70
    BlendContrast=NamedRerouteUsage()
   ) {materialFunction=&MF_BiplanarNormal_Pan}.LocalSpaceNormal) {name=BubbleNormal_1, nodeColor=<rgba 0 1 0.2266 1>}
  n72 = VectorParameter() {group=PrismParticle, parameterName=PP_DiffuseFresnelColor, defaultValue=<rgba 0 0.5731 0.7552 1>}
  n73 = Fresnel() {baseReflectFraction=0, exponent=1}
  n74 = VectorParameter() {group=PrismParticle, parameterName=PP_ColorInterior, defaultValue=<rgba 0 0.0038 0.2344 1>}
  n75 = VectorParameter() {group=PrismParticle, parameterName=PP_Color, defaultValue=<rgba 0.1159 0.2286 1 1>}
  n76 = Fresnel() {exponent=2}
  NamedRerouteDeclaration(None=Add(A=Multiply(A=@n72.RGB, B=@n73), B=LinearInterpolate(A=@n74.RGB, B=@n75.RGB, Alpha=@n76))) {name=BaseColor_OUT, nodeColor=<rgba 0.6641 1 0 1>}
  n77 = PreSkinnedNormal()
  NamedRerouteDeclaration(None=ShaderStageSwitch(PixelShader=VertexInterpolator(VS=@n77).PS, VertexShader=@n77)) {name=PreSkinnedNrm_VS-PS, nodeColor=<rgba 0 0.4844 1 1>}
  n78 = DotProduct(A=PixelNormalWS(), B=CameraVectorWS())
  n79 = Power(
   Base=OneMinus(None=Clamp(None=@n78))
   Exp=ScalarParameter() {defaultValue=2, group=PrismParticle, parameterName=PP_SubSurfaceFresnelExponent}
  )
  n80 = ScalarParameter() {defaultValue=2, group=PrismParticle, parameterName=PP_SubSurfaceOpacity}
  NamedRerouteDeclaration(None=OneMinus(None=Multiply(A=@n79, B=@n80))) {name=SubsurfaceOpacity, nodeColor=<rgba 0 1 0.0859 1>}
  NamedRerouteDeclaration(None=MaterialFunctionCall(
    Texture=NamedRerouteUsage()
    Position=NamedRerouteUsage()
    Normal=NamedRerouteUsage()
    TextureSize=@n57
    Pan=@n58
    BlendContrast=NamedRerouteUsage()
   ) {materialFunction=&MF_BiplanarNormal_Pan}.LocalSpaceNormal) {name=BubbleNormal_3, nodeColor=<rgba 0 1 0.2266 1>}
  NamedRerouteDeclaration(None=@n76) {name=Fresnel, nodeColor=<rgba 1 0 0.9375 1>}
  n81 = PixelNormalWS()
  n82 = Multiply(
   A=ComponentMask(None=Transform(None=@n81) {transformSourceType=TRANSFORMSOURCE_World, transformType=TRANSFORM_View}) {g=true, r=true}
   B=ScalarParameter() {group=PrismParticle, parameterName=PP_InteriorGlowDistortion}
  ) {constB=0.1}
  n83 = Multiply(
   A=SphereMask(
    A=Constant()
    B=Add(A=@n34.UVS, B=@n82)
    Radius=ScalarParameter() {defaultValue=10, group=PrismParticle, parameterName=PP_InternalGlowRadius}
   ) {attenuationRadius=0.5, hardnessPercent=0}
   B=ScalarParameter() {defaultValue=0.8, group=PrismParticle, parameterName=PP_InternalGlowIntensity}
  )
  NamedRerouteDeclaration(None=OneMinus(None=Saturate(None=@n83))) {name=2DGlowGradient, nodeColor=<rgba 1 0 0.9141 1>}
  n84 = Add(A=NamedRerouteUsage(), B=NamedRerouteUsage())
  n85 = NamedRerouteUsage()
  NamedRerouteDeclaration(None=LinearInterpolate(A=Add(A=@n84, B=@n85), Alpha=@n53) {constB=0}) {name=BiplanarBubble, nodeColor=<rgba 0 0.8125 1 1>}
  n86 = CollectionParameter() {
   collection=&FortniteMaterialParameters
   parameterName=SunlightDirectionalVector
  }
  n87 = ComponentMask(None=@n86) {b=true, g=true, r=true}
  n88 = NamedRerouteUsage()
  n89 = Multiply(A=Multiply(A=DotProduct(A=@n87, B=@n88)) {constB=2}, B=@n88)
  n90 = DotProduct(
   A=Normalize(VectorInput=Subtract(A=@n89, B=@n87))
   B=Multiply(A=MaterialFunctionCall() {materialFunction=&CameraDirectionVector}.Result)
  )
  n91 = Power(Base=Max(A=@n90) {constB=0}, Exp=ScalarParameter() {defaultValue=2, group=Slurp, parameterName=Bubble_BackSpecular_Power})
  n92 = ScalarParameter() {defaultValue=0.1, group=Slurp, parameterName=Bubble_BackSpecular_Scale}
  n93 = Multiply(
   A=VectorParameter() {group=Slurp, parameterName=Bubble_BackSpecular_Color, defaultValue=<rgba 1 1 1 1>}.RGB
   B=Saturate(None=Multiply(A=@n91, B=@n92) {constB=0.15})
  )
  n94 = Power(Base=Multiply(A=@n90) {constB=-1}, Exp=ScalarParameter() {defaultValue=30, group=Slurp, parameterName=Bubble_HighlightSpecular_Power})
  n95 = ScalarParameter() {defaultValue=0.8, group=Slurp, parameterName=Bubble_HighlightSpecular_Scale}
  n96 = Multiply(
   A=Saturate(None=Multiply(A=@n94, B=@n95) {constB=0.55})
   B=VectorParameter() {group=Slurp, parameterName=Bubble_HighlightSpecular_Color, defaultValue=<rgba 1 1 1 1>}.RGB
  )
  NamedRerouteDeclaration(None=Multiply(A=Add(A=@n93, B=@n96), B=NamedRerouteUsage())) {name=BubbleHighlight, nodeColor=<rgba 0 0.7891 1 1>}
  NamedRerouteUsage()
  Saturate(None=LinearInterpolate(
    A=ScalarParameter() {defaultValue=0.5, group=PrismParticle, parameterName=PP_Opacity_Outter}
    B=ScalarParameter() {defaultValue=1, group=PrismParticle, parameterName=PP_Opacity_Inner}
    Alpha=OneMinus(None=@n73)
   ))
  StaticComponentMaskParameter(None=@n1.Tech Art Mask) {defaultR=true, group=Skin FX, parameterName=Slurp Channel}
  StaticComponentMaskParameter(None=@n1.Tech Art Mask) {defaultG=true, group=Skin FX, parameterName=Slurp Bubbles Channel}
  StaticComponentMaskParameter(None=@n1.Tech Art Mask) {defaultB=true, group=Skin FX, parameterName=Slurp Normal Distortion Channel}
  StaticSwitchParameter(False=@n16) {parameterName=UseSlurp}
  VertexInterpolator(VS=PreSkinnedPosition())
  VertexInterpolator(VS=PreSkinnedNormal())
 ]
}
